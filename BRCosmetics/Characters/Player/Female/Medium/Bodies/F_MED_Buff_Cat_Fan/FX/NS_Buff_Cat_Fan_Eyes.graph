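# Niagara system that adds glowing eye sprites to the Buff Cat Fan cosmetic character body.
# A single EyeGlow emitter bursts two sprites tinted by User.EffectColor (red by default), split across two spawn groups.
# Each update re-samples User.SkeletalMesh, using the spawn group as bone/socket index and to pick an X-mirrored offset.
# Particles loop their lifetime until the emitter deactivates, with a camera offset applied.
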NiagaraSystem /BRCosmetics/Characters/Player/Female/Medium/Bodies/F_MED_Buff_Cat_Fan/FX/NS_Buff_Cat_Fan_Eyes
dependencies {usedDataInterfaces [], usedDynamicInputs [], usedModules [&SystemState]}
emitters {EyeGlow {
  emitterUpdateScript [
   &EmitterState(Life Cycle Mode=::NewEnumerator1, Loop Behavior=::NewEnumerator1, Loop Duration Mode=::NewEnumerator1, UseLoopDelay=false)
   &SpawnBurst_Instantaneous(Spawn Count=2, Use Spawn Probability=false, UseLoopCountLimit=false)
  ]
  particleSpawnScript [
   &FN_SetSpawnGroup(Number of Spawn Groups=2)
   &InitializeParticle(Color=$User.EffectColor, Sprite Size Mode=::NewEnumerator3, UsePositionOffset=false, Write Material Random=true)
  ]
  particleUpdateScript [
   &CameraOffset(Camera Offset Amount=4)
   &SkeletalMeshLocation(
    Bone / Socket Index=$Particles.SpawnGroup
    MaskSpawnGroup=false
    Mesh Sampling Type=::NewEnumerator5
    Sampled Position Offset=&Lerp_Vector(A=<0.9 1 0.55>, B=<-0.9 1 0.55>, Alpha=&MakeFloatFromInt(Int=$Particles.SpawnGroup))
    Skeletal Mesh=$User.SkeletalMesh
    Socket Sampling Mode=::NewEnumerator5
   )
   &ParticleState(
    Kill Particles When Lifetime Has Elapsed=false
    Let Infinitely Lived Particles Die When Emitter Deactivates=true
    Loop Particles Lifetime=true
   )
  ]
  rendererClasses [&NiagaraSpriteRendererProperties]
  renderers [{rendererClass=&NiagaraSpriteRendererProperties}]
 }}
summary {
 systemName=NS_Buff_Cat_Fan_Eyes
 emitters [{emitterName=EyeGlow, rendererClasses [&NiagaraSpriteRendererProperties]}]
 userVariables [
  User.bIsFrontEndPreview {type=NiagaraBool}
  User.EffectColor {type=LinearColor, description="", defaultValue {struct=LinearColor, value=<rgba 1 0.05 0 0.25>}}
  User.SkeletalMesh {
   type=NiagaraDataInterfaceSkeletalMesh
   defaultValue {
    struct=NiagaraExt_VariableValue_DataInterface
    value {dataInterface=&self:NiagaraDataInterfaceSkeletalMesh_0}
   }
  }
 ]
}
