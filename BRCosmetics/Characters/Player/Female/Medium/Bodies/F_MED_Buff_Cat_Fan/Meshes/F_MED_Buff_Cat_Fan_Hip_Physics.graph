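# Physics asset for the hip mesh of the F_MED_Buff_Cat_Fan female player cosmetic, set up for dangling mask, bandana, pouch, side camera belt and shoe fins.
# Capsules on pelvis, thighs and feet form the body, linked by free-swinging constraints.
# Mask, bandana and two-link pouch chains hang from tiny master bodies on the pelvis; the belt attaches directly, the fins to the feet.
# Accessory joints have narrow twist limits, and fins swing mostly on one axis.
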PhysicsAsset /BRCosmetics/Characters/Player/Female/Medium/Bodies/F_MED_Buff_Cat_Fan/Meshes/F_MED_Buff_Cat_Fan_Hip_Physics
bodies {
 Dyn_Bandana1 [{shapeType=Capsule, length=3, radius=2, shapeName=Dyn_Bandana1_capsule, center=<3 0 0>, rotation=<rot 90 0 0>}]
 Dyn_CameraBelt_Side [
  {
   shapeType=Box
   extentX=4
   extentY=1
   extentZ=10.5
   shapeName=Dyn_CameraBelt_Side_box
   center=<-0.0263 -0.2241 0.06>
   rotation=<rot 0 -6.6823 0>
  }
  {
   shapeType=Box
   extentX=3
   extentY=0.5
   extentZ=6
   shapeName=Dyn_CameraBelt_Side_box
   center=<-0.4109 1.1621 -5.4847>
   rotation=<rot -4.7624 -4.1668 -48.733>
  }
 ]
 Dyn_Mask2 [
  {
   shapeType=Capsule
   length=8
   radius=2
   shapeName=Dyn_Mask2_capsule
   center=<2.9412 -1.4644 -2.4642>
   rotation=<rot 79.9962 -175.0004 -179.9999>
  }
  {shapeType=Capsule, length=8, radius=2, shapeName=Dyn_Mask2_capsule, center=<1.8262 -1.5632 5.3292>, rotation=<rot 79.992 -175 -180>}
  {shapeType=Capsule, length=8, radius=2, shapeName=Dyn_Mask2_capsule, center=<2.1796 -2.3826 1.2759>, rotation=<rot 79.9907 -180 -180>}
  {shapeType=Capsule, length=6, radius=2, shapeName=Dyn_Mask2_capsule, center=<-4.3027 -1.3225 0.0104>, rotation=<rot 9.8468 1.7539 10.1522>}
 ]
 Dyn_Master_Bandana [{shapeType=Capsule, length=0.5, radius=0.5, shapeName=Dyn_Master_Bandana_capsule}]
 Dyn_Master_Mask [{shapeType=Capsule, length=0.5, radius=0.5, shapeName=Dyn_Master_Mask_capsule}]
 Dyn_Master_Pouch [{shapeType=Capsule, length=0.5, radius=0.5, shapeName=Dyn_Master_Pouch_capsule, rotation=<rot 0 -90 0>}]
 Dyn_Pouch2 [{shapeType=Capsule, length=4, radius=2, shapeName=Dyn_Pouch2_capsule, center=<1 -4.1812 0>, rotation=<rot 90 0 15.0016>}]
 Dyn_Pouch3 [
  {shapeType=Sphere, radius=5, shapeName=Dyn_Pouch3_sphere, center=<5.7934 -5.815 0.0001>}
  {shapeType=Capsule, length=5, radius=2, shapeName=Dyn_Pouch3_capsule, center=<5.0233 -4.4109 0.0002>, rotation=<rot 90 0 0>}
  {shapeType=Capsule, length=3.25, radius=4, shapeName=Dyn_Pouch3_capsule, center=<7.0296 -3.622 -0.0916>}
 ]
 Dyn_ShoeFin_l [{shapeType=Capsule, length=2, radius=1, shapeName=Dyn_ShoeFin_l_capsule, center=<1 0 0>, rotation=<rot 0 0 90>}]
 Dyn_ShoeFin_r [{shapeType=Capsule, length=2, radius=1, shapeName=Dyn_ShoeFin_r_capsule, center=<1 0 0>, rotation=<rot 0 0 90>}]
 foot_l [{shapeType=Capsule, length=0.5, radius=0.5, shapeName=foot_l_capsule}]
 foot_r [{shapeType=Capsule, length=0.5, radius=0.5, shapeName=foot_r_capsule}]
 pelvis [
  {shapeType=Capsule, length=15, radius=8, shapeName=pelvis_capsule, center=<-2.9528 -1.6237 0>}
  {shapeType=Capsule, length=15, radius=6.5, shapeName=pelvis_capsule, center=<2.7 0.3476 0>}
  {shapeType=Capsule, length=10.25, radius=9.75, shapeName=pelvis_capsule, center=<-5.0307 -1.0901 0.0464>}
 ]
 thigh_l [{shapeType=Capsule, length=16.6619, radius=6.5, shapeName=thigh_l_capsule, center=<-13.4862 -5.1989 -0.2437>, rotation=<rot 90 0 0>}]
 thigh_r [
  {
   shapeType=Capsule
   length=20.1582
   radius=6.6067
   shapeName=thigh_r_capsule
   center=<16.2314 -1.0659 0.0002>
   rotation=<rot -0.2844 -88.7783 -92.1266>
  }
  {
   shapeType=Capsule
   length=26.75
   radius=7
   shapeName=thigh_r_capsule
   center=<17.8347 2.7663 -0.1841>
   rotation=<rot -88.125 -2.4023 -2.8354>
  }
 ]
 thigh_twist_01_l [{
   shapeType=Capsule
   length=20
   radius=7
   shapeName=thigh_twist_01_l_capsule
   center=<12.6275 -0.1138 0.2885>
   rotation=<rot 90 12.9074 7.9075>
  }]
}
constraints [
 {
  bone1Name=thigh_twist_01_l
  bone2Name=thigh_l
  swing1LimitDegrees=45
  swing1Motion=Free
  swing2LimitDegrees=45
  swing2Motion=Free
  twistLimitDegrees=45
  twistMotion=Free
 }
 {
  bone1Name=thigh_l
  bone2Name=pelvis
  swing1LimitDegrees=45
  swing1Motion=Free
  swing2LimitDegrees=45
  swing2Motion=Free
  twistLimitDegrees=45
  twistMotion=Free
 }
 {
  bone1Name=Dyn_Master_Mask
  bone2Name=pelvis
  swing1LimitDegrees=45
  swing1Motion=Free
  swing2LimitDegrees=45
  swing2Motion=Free
  twistLimitDegrees=45
  twistMotion=Free
 }
 {bone1Name=Dyn_Mask2, bone2Name=Dyn_Master_Mask, swing1LimitDegrees=45, swing2LimitDegrees=20, twistLimitDegrees=1}
 {
  bone1Name=Dyn_Master_Bandana
  bone2Name=pelvis
  swing1LimitDegrees=45
  swing1Motion=Free
  swing2LimitDegrees=45
  swing2Motion=Free
  twistLimitDegrees=45
  twistMotion=Free
 }
 {bone1Name=Dyn_Bandana1, bone2Name=Dyn_Master_Bandana, swing1LimitDegrees=40, swing2LimitDegrees=40, twistLimitDegrees=1}
 {
  bone1Name=Dyn_Master_Pouch
  bone2Name=pelvis
  swing1LimitDegrees=45
  swing1Motion=Free
  swing2LimitDegrees=45
  swing2Motion=Free
  twistLimitDegrees=45
  twistMotion=Free
 }
 {bone1Name=Dyn_Pouch2, bone2Name=Dyn_Master_Pouch, swing1LimitDegrees=45, swing2LimitDegrees=20, twistLimitDegrees=1}
 {bone1Name=Dyn_Pouch3, bone2Name=Dyn_Pouch2, swing1LimitDegrees=45, swing2LimitDegrees=20, twistLimitDegrees=1}
 {
  bone1Name=foot_r
  bone2Name=thigh_r
  swing1LimitDegrees=45
  swing1Motion=Free
  swing2LimitDegrees=45
  swing2Motion=Free
  twistLimitDegrees=45
  twistMotion=Free
 }
 {
  bone1Name=foot_l
  bone2Name=thigh_l
  swing1LimitDegrees=45
  swing1Motion=Free
  swing2LimitDegrees=45
  swing2Motion=Free
  twistLimitDegrees=45
  twistMotion=Free
 }
 {bone1Name=Dyn_ShoeFin_r, bone2Name=foot_r, swing1LimitDegrees=1, swing2LimitDegrees=20, twistLimitDegrees=1}
 {bone1Name=Dyn_ShoeFin_l, bone2Name=foot_l, swing1LimitDegrees=1, swing2LimitDegrees=20, twistLimitDegrees=1}
 {bone1Name=Dyn_CameraBelt_Side, bone2Name=pelvis, swing1LimitDegrees=20, swing2LimitDegrees=20, twistLimitDegrees=1}
 {
  bone1Name=thigh_r
  bone2Name=pelvis
  swing1LimitDegrees=45
  swing1Motion=Free
  swing2LimitDegrees=45
  swing2Motion=Free
  twistLimitDegrees=45
  twistMotion=Free
 }
]
