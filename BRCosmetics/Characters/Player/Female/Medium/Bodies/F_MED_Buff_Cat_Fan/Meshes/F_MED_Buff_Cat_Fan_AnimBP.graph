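# Anim Blueprint for the F_MED_Buff_Cat_Fan cosmetic body that repositions its camera, belt, pouch and bead accessory bones to suit player movement.
# It copies the attached parent's pose, then layers additive bone-space Modify Bone offsets, some weighted by crouch-walk, crouch-sprint and leg-movement variables.
# Bool blends select between crouch and sprint variants.
# When bRigidBodyEnabled is true, the output adds weighted pouch constraints to pelvis, thigh and spine bones over the cached pose.
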AnimBlueprint /BRCosmetics/Characters/Player/Female/Medium/Bodies/F_MED_Buff_Cat_Fan/Meshes/F_MED_Buff_Cat_Fan_AnimBP
target_skeleton=&/F_MED_Buff_Cat_Fan_Skeleton
graphs [
 AnimGraph {nodes [
   Component To Local()
   Component To Local()
   "Constraint - Dyn_Pouch1 (PPPPPPP)"!(
    ComponentPose
    bAlphaBoolEnabled:bool=True
    ConstraintWeights_0=0.0026
    ConstraintWeights_1:float=0.0044
    ConstraintWeights_2:float=0.0007
    ConstraintWeights_3:float=0
    ConstraintWeights_4:float=0
    ConstraintWeights_5:float=0.0518
    ConstraintWeights_6:float=0.9404
   ) -> (Pose) {
    alphaInputType=Bool
    alphaBoolBlend {blendOption=Linear}
    boneToModify {boneName=Dyn_Pouch1}
    constraintSetup [
     {targetBone {boneName=spine_01}}
     {targetBone {boneName=spine_02}}
     {targetBone {boneName=spine_04}}
     {targetBone {boneName=thigh_l}}
     {targetBone {boneName=thigh_twist_01_r}}
     {targetBone {boneName=thigh_twist_01_l}}
     {targetBone {boneName=pelvis}}
    ]
    constraintWeights [1,1,1,1,1,1,1]
   }
   Local To Component(LocalPose=Blend Poses by bool(BlendPose_0=Component To Local(), BlendPose_1=Component To Local(), BlendTime_0=0.2, BlendTime_1=0.2, bActiveValue=Get bIsCrouchSprinting()) {blendType=Linear})
   n1 = "Transform (Modify) Bone - Bone: Dyn_Master_Pouch"(
    ComponentPose=Local To Component()
    Alpha=Get Left_Leg Movement() -> (Left_Leg Movement:double)
    Translation="-2,0,-2"
    Rotation="0,0,0"
    Scale="1,1,1"
   ) {
    rotationMode=BMM_Additive
    rotationSpace=BCS_BoneSpace
    scaleMode=BMM_Additive
    scaleSpace=BCS_BoneSpace
    translationMode=BMM_Additive
    translationSpace=BCS_BoneSpace
    alphaBoolBlend {blendInTime=0.05, blendOption=Linear, blendOutTime=0.1}
    alphaScaleBiasClamp {interpSpeedDecreasing=1, interpSpeedIncreasing=0.5}
    boneToModify {boneName=Dyn_Master_Pouch}
    translation {x=-2, z=-2}
   }
   n2 = "Transform (Modify) Bone - Bone: Dyn_Master_Pouch"(ComponentPose=Local To Component(), Alpha=1, Translation="-4,0,-3", Rotation="0,-20,-50", Scale="1,1,1") {
    rotationMode=BMM_Additive
    rotationSpace=BCS_BoneSpace
    scaleMode=BMM_Additive
    scaleSpace=BCS_BoneSpace
    translationMode=BMM_Additive
    translationSpace=BCS_BoneSpace
    alphaBoolBlend {blendInTime=0.05, blendOption=Linear, blendOutTime=0.1}
    alphaScaleBiasClamp {interpSpeedDecreasing=1, interpSpeedIncreasing=0.5}
    boneToModify {boneName=Dyn_Master_Pouch}
    rotation {roll=-50, yaw=-20}
    translation {x=-4, z=-3}
   }
   n3 = "Transform (Modify) Bone - Bone: Dyn_Camera"(ComponentPose=@n2, Alpha=1, Translation="-1,0,0", Rotation="0,-5,10", Scale="1,1,1") {
    rotationMode=BMM_Additive
    rotationSpace=BCS_BoneSpace
    scaleMode=BMM_Additive
    scaleSpace=BCS_BoneSpace
    translationMode=BMM_Additive
    translationSpace=BCS_BoneSpace
    alphaBoolBlend {blendInTime=0.2, blendOption=Linear, blendOutTime=0.2}
    alphaScaleBiasClamp {bInterpResult=true, interpSpeedDecreasing=5, interpSpeedIncreasing=5}
    boneToModify {boneName=Dyn_Camera}
    rotation {roll=10, yaw=-5}
    translation {x=-1}
   }
   n4 = Get bIsCrouchSprinting()
   n5 = "Transform (Modify) Bone - Bone: Dyn_Camera"!(ComponentPose=@n3, bAlphaBoolEnabled:bool=@n4, Translation="0,-2,0", Rotation="0,0,0", Scale="1,1,1") -> (Pose) {
    alphaInputType=Bool
    rotationMode=BMM_Additive
    rotationSpace=BCS_BoneSpace
    scaleMode=BMM_Additive
    scaleSpace=BCS_BoneSpace
    translationMode=BMM_Additive
    translationSpace=BCS_BoneSpace
    alphaBoolBlend {blendInTime=0.2, blendOption=Linear, blendOutTime=0.2}
    alphaScaleBiasClamp {interpSpeedDecreasing=0.5, interpSpeedIncreasing=0.5}
    boneToModify {boneName=Dyn_Camera}
    translation {y=-2}
   }
   n6 = "Transform (Modify) Bone - Bone: Dyn_Master_Belt"!(ComponentPose=@n5, bAlphaBoolEnabled:bool=@n4, Translation="0,0,0", Rotation="10,0,0", Scale="1,1,1") -> (Pose) {
    alphaInputType=Bool
    rotationMode=BMM_Additive
    rotationSpace=BCS_BoneSpace
    scaleMode=BMM_Additive
    scaleSpace=BCS_BoneSpace
    translationMode=BMM_Replace
    translationSpace=BCS_BoneSpace
    alphaBoolBlend {blendInTime=0.2, blendOption=Linear, blendOutTime=0.2}
    alphaScaleBiasClamp {interpSpeedDecreasing=5, interpSpeedIncreasing=2}
    boneToModify {boneName=Dyn_Master_Belt}
    rotation {pitch=10}
   }
   n7 = "Transform (Modify) Bone - Bone: Dyn_CameraBelt_Side"(ComponentPose=@n6, Alpha=1, Translation="-1,0,0", Rotation="0,-15,0", Scale="1,1,1") {
    rotationMode=BMM_Additive
    rotationSpace=BCS_BoneSpace
    scaleMode=BMM_Additive
    scaleSpace=BCS_BoneSpace
    translationMode=BMM_Additive
    translationSpace=BCS_BoneSpace
    alphaBoolBlend {blendInTime=0.15, blendOption=Linear, blendOutTime=1}
    alphaScaleBiasClamp {interpSpeedDecreasing=1, interpSpeedIncreasing=0.5}
    boneToModify {boneName=Dyn_CameraBelt_Side}
    rotation {yaw=-15}
    translation {x=-1}
   }
   n8 = "Transform (Modify) Bone - Bone: Dyn_Master_Belt_RightFront"(ComponentPose=@n7, Alpha=1, Translation="-2,0,0", Rotation="0,0,0", Scale="1,1,1") {
    rotationMode=BMM_Additive
    rotationSpace=BCS_BoneSpace
    scaleMode=BMM_Additive
    scaleSpace=BCS_BoneSpace
    translationMode=BMM_Additive
    translationSpace=BCS_BoneSpace
    alphaBoolBlend {blendInTime=0.15, blendOption=Linear, blendOutTime=1}
    alphaScaleBiasClamp {interpSpeedDecreasing=1, interpSpeedIncreasing=0.5}
    boneToModify {boneName=Dyn_Master_Belt_RightFront}
    translation {x=-2}
   }
   Local To Component(LocalPose=Blend Poses by bool(BlendPose_0=Component To Local(ComponentPose=@n8), BlendPose_1=Component To Local(ComponentPose=@n1), bActiveValue=Get bIsCrouching()) {blendType=Linear})
   n9 = "Transform (Modify) Bone - Bone: Dyn_Master_LowerBelt"(Alpha=1, Translation="0,0,0", Rotation="-25,0,0", Scale="1,1,1") {
    rotationMode=BMM_Additive
    rotationSpace=BCS_BoneSpace
    scaleMode=BMM_Additive
    scaleSpace=BCS_BoneSpace
    translationMode=BMM_Additive
    translationSpace=BCS_BoneSpace
    alphaBoolBlend {blendInTime=0.15, blendOption=Linear, blendOutTime=1}
    alphaScaleBiasClamp {interpSpeedDecreasing=1, interpSpeedIncreasing=20}
    boneToModify {boneName=Dyn_Master_LowerBelt}
    rotation {pitch=-25}
   }
   n10 = "Transform (Modify) Bone - Bone: Dyn_Camera"!(ComponentPose=@n9, Alpha:float=1, Translation="-5,0,0", Rotation="0,-6,20", Scale="1,1,1") -> (Pose) {
    rotationMode=BMM_Additive
    rotationSpace=BCS_BoneSpace
    scaleMode=BMM_Additive
    scaleSpace=BCS_BoneSpace
    translationMode=BMM_Additive
    translationSpace=BCS_BoneSpace
    alphaBoolBlend {blendInTime=0.15, blendOption=Linear, blendOutTime=1}
    alphaScaleBiasClamp {interpSpeedDecreasing=1, interpSpeedIncreasing=20}
    boneToModify {boneName=Dyn_Camera}
    rotation {roll=20, yaw=-6}
    translation {x=-5}
   }
   n11 = "Transform (Modify) Bone - Bone: Dyn_CameraBelt_Side"(ComponentPose=@n10, Alpha=1, Translation="-3,0.8,0", Rotation="-25,0,0", Scale="1,1,1") {
    rotationMode=BMM_Additive
    rotationSpace=BCS_BoneSpace
    scaleMode=BMM_Additive
    scaleSpace=BCS_BoneSpace
    translationMode=BMM_Additive
    translationSpace=BCS_BoneSpace
    alphaBoolBlend {blendInTime=0.15, blendOption=Linear, blendOutTime=1}
    alphaScaleBiasClamp {interpSpeedDecreasing=1, interpSpeedIncreasing=20}
    boneToModify {boneName=Dyn_CameraBelt_Side}
    rotation {pitch=-25}
    translation=<-3 0.8>
   }
   "Transform (Modify) Bone - Bone: Dyn_Master_Belt_RightFront"(ComponentPose=@n11, Alpha=1, Translation="-3,-0.5,0", Rotation="20,0,-10", Scale="1,1,1") {
    rotationMode=BMM_Additive
    rotationSpace=BCS_BoneSpace
    scaleMode=BMM_Additive
    scaleSpace=BCS_BoneSpace
    translationMode=BMM_Additive
    translationSpace=BCS_BoneSpace
    alphaBoolBlend {blendInTime=0.15, blendOption=Linear, blendOutTime=1}
    alphaScaleBiasClamp {interpSpeedDecreasing=1, interpSpeedIncreasing=20}
    boneToModify {boneName=Dyn_Master_Belt_RightFront}
    rotation {pitch=20, roll=-10}
    translation=<-3 -0.5>
   }
   "Transform (Modify) Bone - Bone: Dyn_CameraBelt_Side"(Alpha=1, Translation="-2,0.8,0", Rotation="-20,-5,20", Scale="1,1,1") {
    rotationMode=BMM_Additive
    rotationSpace=BCS_BoneSpace
    scaleMode=BMM_Additive
    scaleSpace=BCS_BoneSpace
    translationMode=BMM_Additive
    translationSpace=BCS_BoneSpace
    alphaBoolBlend {blendInTime=0.15, blendOption=Linear, blendOutTime=1}
    alphaScaleBiasClamp {interpSpeedDecreasing=1, interpSpeedIncreasing=20}
    boneToModify {boneName=Dyn_CameraBelt_Side}
    rotation=<rot -20 -5 20>
    translation=<-2 0.8>
   }
   n12 = "Transform (Modify) Bone - Bone: Dyn_Master_LowerBelt"(Alpha=1, Translation="0,0,0", Rotation="-25,0,0", Scale="1,1,1") {
    rotationMode=BMM_Additive
    rotationSpace=BCS_BoneSpace
    scaleMode=BMM_Additive
    scaleSpace=BCS_BoneSpace
    translationMode=BMM_Additive
    translationSpace=BCS_BoneSpace
    alphaBoolBlend {blendInTime=0.15, blendOption=Linear, blendOutTime=1}
    alphaScaleBiasClamp {interpSpeedDecreasing=1, interpSpeedIncreasing=20}
    boneToModify {boneName=Dyn_Master_LowerBelt}
    rotation {pitch=-25}
   }
   n13 = "Transform (Modify) Bone - Bone: Dyn_Camera"(ComponentPose=@n12, Alpha=1, Translation="-2.5,4,4", Rotation="-10.000,-0.000,60", Scale="1,1,1") {
    rotationMode=BMM_Additive
    rotationSpace=BCS_BoneSpace
    scaleMode=BMM_Additive
    scaleSpace=BCS_BoneSpace
    translationMode=BMM_Additive
    translationSpace=BCS_BoneSpace
    alphaBoolBlend {blendInTime=0.15, blendOption=Linear, blendOutTime=1}
    alphaScaleBiasClamp {interpSpeedDecreasing=1, interpSpeedIncreasing=20}
    boneToModify {boneName=Dyn_Camera}
    rotation {pitch=-10, roll=60}
    translation=<-2.5 4 4>
   }
   n14 = "Transform (Modify) Bone - Bone: Dyn_CameraBelt_Side"(ComponentPose=@n13, Alpha=1, Translation="-2,0.8,0", Rotation="-20,-5,20", Scale="1,1,1") {
    rotationMode=BMM_Additive
    rotationSpace=BCS_BoneSpace
    scaleMode=BMM_Additive
    scaleSpace=BCS_BoneSpace
    translationMode=BMM_Additive
    translationSpace=BCS_BoneSpace
    alphaBoolBlend {blendInTime=0.15, blendOption=Linear, blendOutTime=1}
    alphaScaleBiasClamp {interpSpeedDecreasing=1, interpSpeedIncreasing=20}
    boneToModify {boneName=Dyn_CameraBelt_Side}
    rotation=<rot -20 -5 20>
    translation=<-2 0.8>
   }
   n15 = "Transform (Modify) Bone - Bone: Dyn_Master_Belt_RightFront"(ComponentPose=@n14, Alpha=1, Translation="-3,0,0", Rotation="20,0,0", Scale="1,1,1") {
    rotationMode=BMM_Additive
    rotationSpace=BCS_BoneSpace
    scaleMode=BMM_Additive
    scaleSpace=BCS_BoneSpace
    translationMode=BMM_Additive
    translationSpace=BCS_BoneSpace
    alphaBoolBlend {blendInTime=0.15, blendOption=Linear, blendOutTime=1}
    alphaScaleBiasClamp {interpSpeedDecreasing=1, interpSpeedIncreasing=20}
    boneToModify {boneName=Dyn_Master_Belt_RightFront}
    rotation {pitch=20}
    translation {x=-3}
   }
   n16 = "Transform (Modify) Bone - Bone: Dyn_CameraBelt_Back"(ComponentPose=@n15, Alpha=1, Translation="-2,0,0", Rotation="10,0,0", Scale="1,1,1") {
    rotationMode=BMM_Additive
    rotationSpace=BCS_BoneSpace
    scaleMode=BMM_Additive
    scaleSpace=BCS_BoneSpace
    translationMode=BMM_Additive
    translationSpace=BCS_BoneSpace
    alphaBoolBlend {blendInTime=0.15, blendOption=Linear, blendOutTime=1}
    alphaScaleBiasClamp {interpSpeedDecreasing=1, interpSpeedIncreasing=20}
    boneToModify {boneName=Dyn_CameraBelt_Back}
    rotation {pitch=10}
    translation {x=-2}
   }
   "Transform (Modify) Bone - Bone: Dyn_Master_Belt_Rightback"(ComponentPose=@n16, Alpha=1, Translation="-1,-0.5,0", Rotation="-5,0,0", Scale="1,1,1") {
    rotationMode=BMM_Additive
    rotationSpace=BCS_BoneSpace
    scaleMode=BMM_Additive
    scaleSpace=BCS_BoneSpace
    translationMode=BMM_Additive
    translationSpace=BCS_BoneSpace
    alphaBoolBlend {blendInTime=0.15, blendOption=Linear, blendOutTime=1}
    alphaScaleBiasClamp {interpSpeedDecreasing=1, interpSpeedIncreasing=20}
    boneToModify {boneName=Dyn_Master_Belt_Rightback}
    rotation {pitch=-5}
    translation=<-1 -0.5>
   }
   "Transform (Modify) Bone - Bone: Dyn_Camera"(Alpha=1, Translation="-12,10,20", Rotation="0,0,60", Scale="1,1,1") {
    rotationMode=BMM_Additive
    rotationSpace=BCS_BoneSpace
    scaleMode=BMM_Additive
    scaleSpace=BCS_BoneSpace
    translationMode=BMM_Additive
    translationSpace=BCS_BoneSpace
    alphaBoolBlend {blendInTime=0.15, blendOption=Linear, blendOutTime=1}
    alphaScaleBiasClamp {interpSpeedDecreasing=1, interpSpeedIncreasing=20}
    boneToModify {boneName=Dyn_Camera}
    rotation {roll=60}
    translation=<-12 10 20>
   }
   "Transform (Modify) Bone - Bone: Dyn_Master_Mask"(Alpha=0.5, Translation="0,-1,0", Rotation="0,0,0", Scale="1,1,1") {
    rotationMode=BMM_Additive
    rotationSpace=BCS_BoneSpace
    scaleMode=BMM_Additive
    scaleSpace=BCS_BoneSpace
    translationMode=BMM_Additive
    translationSpace=BCS_BoneSpace
    alphaBoolBlend {blendInTime=0.15, blendOption=Linear, blendOutTime=1}
    alphaScaleBiasClamp {bInterpResult=true, interpSpeedDecreasing=2, interpSpeedIncreasing=2}
    boneToModify {boneName=Dyn_Master_Mask}
    translation {y=-1}
   }
   "Transform (Modify) Bone - Bone: Dyn_Pouch2"(Alpha=1, Translation="0,0,0", Rotation="0,0,0", Scale="1,1,1") {
    rotationMode=BMM_Additive
    rotationSpace=BCS_BoneSpace
    scaleMode=BMM_Additive
    scaleSpace=BCS_BoneSpace
    translationMode=BMM_Additive
    translationSpace=BCS_BoneSpace
    alphaBoolBlend {blendInTime=0.05, blendOption=Linear, blendOutTime=0.1}
    alphaScaleBiasClamp {interpSpeedDecreasing=1, interpSpeedIncreasing=0.5}
    boneToModify {boneName=Dyn_Pouch2}
   }
   "Transform (Modify) Bone - Bone: Dyn_CameraBelt_Side"!(ComponentPose, bAlphaBoolEnabled:bool=Get bIsInFrontEnd(), Translation="-3,0,0", Rotation="-30,0,0", Scale="1,1,1") -> (Pose) {
    alphaInputType=Bool
    rotationMode=BMM_Additive
    rotationSpace=BCS_BoneSpace
    scaleMode=BMM_Additive
    scaleSpace=BCS_BoneSpace
    translationMode=BMM_Additive
    translationSpace=BCS_BoneSpace
    alphaBoolBlend {blendInTime=0.15, blendOption=Linear, blendOutTime=1}
    alphaScaleBiasClamp {interpSpeedDecreasing=1, interpSpeedIncreasing=20}
    boneToModify {boneName=Dyn_CameraBelt_Side}
    rotation {pitch=-30}
    translation {x=-3}
   }
   n17 = "Transform (Modify) Bone - Bone: Dyn_Camera"(Alpha=0, Translation="-2,5,4", Rotation="0,-30,60", Scale="1,1,1") {
    rotationMode=BMM_Additive
    rotationSpace=BCS_BoneSpace
    scaleMode=BMM_Additive
    scaleSpace=BCS_BoneSpace
    translationMode=BMM_Additive
    translationSpace=BCS_BoneSpace
    alphaBoolBlend {blendInTime=0.2, blendOption=Linear, blendOutTime=0.2}
    alphaScaleBiasClamp {bInterpResult=true}
    boneToModify {boneName=Dyn_Camera}
    rotation {roll=60, yaw=-30}
    translation=<-2 5 4>
   }
   n18 = "Transform (Modify) Bone - Bone: Dyn_CameraBelt_Side"!(ComponentPose=@n17, Alpha:float=0, Translation="-2,0,0", Rotation="0,0,0", Scale="1,1,1") -> (Pose) {
    rotationMode=BMM_Additive
    rotationSpace=BCS_BoneSpace
    scaleMode=BMM_Additive
    scaleSpace=BCS_BoneSpace
    translationMode=BMM_Additive
    translationSpace=BCS_BoneSpace
    alphaBoolBlend {blendInTime=0.15, blendOption=Linear, blendOutTime=1}
    alphaScaleBiasClamp {bInterpResult=true}
    boneToModify {boneName=Dyn_CameraBelt_Side}
    translation {x=-2}
   }
   n19 = "Transform (Modify) Bone - Bone: Dyn_CameraBelt_Back"(ComponentPose=@n18, Alpha=0, Translation="3,1,0", Rotation="0,0,0", Scale="1,1,1") {
    rotationMode=BMM_Additive
    rotationSpace=BCS_BoneSpace
    scaleMode=BMM_Additive
    scaleSpace=BCS_BoneSpace
    translationMode=BMM_Additive
    translationSpace=BCS_BoneSpace
    alphaBoolBlend {blendInTime=0.15, blendOption=Linear, blendOutTime=1}
    alphaScaleBiasClamp {bInterpResult=true}
    boneToModify {boneName=Dyn_CameraBelt_Back}
    translation=<3 1>
   }
   n20 = "Transform (Modify) Bone - Bone: Dyn_Master_Belt_RightFront"(ComponentPose=@n19, Alpha=0, Translation="-2,0,0", Rotation="0,0,0", Scale="1,1,1") {
    rotationMode=BMM_Additive
    rotationSpace=BCS_BoneSpace
    scaleMode=BMM_Additive
    scaleSpace=BCS_BoneSpace
    translationMode=BMM_Additive
    translationSpace=BCS_BoneSpace
    alphaBoolBlend {blendInTime=0.15, blendOption=Linear, blendOutTime=1}
    alphaScaleBiasClamp {bInterpResult=true}
    boneToModify {boneName=Dyn_Master_Belt_RightFront}
    translation {x=-2}
   }
   n21 = Get bIsSprinting()
   n22 = "Transform (Modify) Bone - Bone: Dyn_Master_Belt"(ComponentPose=@n20, bAlphaBoolEnabled=@n21, Translation="0,0,0", Rotation="2,0,0", Scale="1,1,1") {
    alphaInputType=Bool
    rotationMode=BMM_Additive
    rotationSpace=BCS_BoneSpace
    scaleMode=BMM_Additive
    scaleSpace=BCS_BoneSpace
    translationMode=BMM_Replace
    translationSpace=BCS_BoneSpace
    alphaBoolBlend {blendInTime=1, blendOption=Linear, blendOutTime=1}
    alphaScaleBiasClamp {interpSpeedDecreasing=5, interpSpeedIncreasing=2}
    boneToModify {boneName=Dyn_Master_Belt}
    rotation {pitch=2}
   }
   n23 = "Transform (Modify) Bone - Bone: Dyn_Camera"!(ComponentPose=@n22, bAlphaBoolEnabled:bool=@n21, Translation="0,1,0", Rotation="0,0,20", Scale="1,1,1") -> (Pose) {
    alphaInputType=Bool
    rotationMode=BMM_Additive
    rotationSpace=BCS_BoneSpace
    scaleMode=BMM_Additive
    scaleSpace=BCS_BoneSpace
    translationMode=BMM_Additive
    translationSpace=BCS_BoneSpace
    alphaBoolBlend {blendInTime=1, blendOption=Linear, blendOutTime=1}
    alphaScaleBiasClamp {bInterpResult=true}
    boneToModify {boneName=Dyn_Camera}
    rotation {roll=20}
    translation {y=1}
   }
   "Transform (Modify) Bone - Bone: Dyn_CameraBelt_Side"!(ComponentPose=@n23, bAlphaBoolEnabled:bool=@n21, Translation="0,1,0", Rotation="0,0,0", Scale="1,1,1") -> (Pose) {
    alphaInputType=Bool
    rotationMode=BMM_Additive
    rotationSpace=BCS_BoneSpace
    scaleMode=BMM_Additive
    scaleSpace=BCS_BoneSpace
    translationMode=BMM_Additive
    translationSpace=BCS_BoneSpace
    alphaBoolBlend {blendInTime=0.15, blendOption=Linear, blendOutTime=1}
    alphaScaleBiasClamp {bInterpResult=true, interpSpeedDecreasing=1, interpSpeedIncreasing=1}
    boneToModify {boneName=Dyn_CameraBelt_Side}
    translation {y=1}
   }
   "Transform (Modify) Bone - Bone: Dyn_Camera"!(ComponentPose, Alpha:float=0, Translation="-2,5,4", Rotation="0,-30,60", Scale="1,1,1") -> (Pose) {
    rotationMode=BMM_Additive
    rotationSpace=BCS_BoneSpace
    scaleMode=BMM_Additive
    scaleSpace=BCS_BoneSpace
    translationMode=BMM_Additive
    translationSpace=BCS_BoneSpace
    alphaBoolBlend {blendInTime=0.2, blendOption=Linear, blendOutTime=0.2}
    alphaScaleBiasClamp {bInterpResult=true}
    boneToModify {boneName=Dyn_Camera}
    rotation {roll=60, yaw=-30}
    translation=<-2 5 4>
   }
   n24 = "Transform (Modify) Bone - Bone: Dyn_Master_Belt"(bAlphaBoolEnabled=True, Translation="0,0,0", Rotation="-15,0,0", Scale="1,1,1") {
    alphaInputType=Bool
    rotationMode=BMM_Additive
    rotationSpace=BCS_BoneSpace
    scaleMode=BMM_Additive
    scaleSpace=BCS_BoneSpace
    translationMode=BMM_Additive
    translationSpace=BCS_BoneSpace
    alphaBoolBlend {blendInTime=1, blendOption=Linear, blendOutTime=1}
    alphaScaleBiasClamp {bInterpResult=true}
    boneToModify {boneName=Dyn_Master_Belt}
    rotation {pitch=-15}
   }
   "Transform (Modify) Bone - Bone: Dyn_Camera"!(ComponentPose=@n24, bAlphaBoolEnabled:bool=True, Translation="0,0,0", Rotation="0,10,60", Scale="1,1,1") -> (Pose) {
    alphaInputType=Bool
    rotationMode=BMM_Additive
    rotationSpace=BCS_BoneSpace
    scaleMode=BMM_Additive
    scaleSpace=BCS_BoneSpace
    translationMode=BMM_Additive
    translationSpace=BCS_BoneSpace
    alphaBoolBlend {blendInTime=1, blendOption=Linear, blendOutTime=1}
    alphaScaleBiasClamp {bInterpResult=true}
    boneToModify {boneName=Dyn_Camera}
    rotation {roll=60, yaw=10}
   }
   "Transform (Modify) Bone - Bone: Dyn_Master_Belt"!(ComponentPose, Alpha:float=0, Translation="0,0,0", Rotation="-20,0,0", Scale="1,1,1") -> (Pose) {
    rotationMode=BMM_Additive
    rotationSpace=BCS_BoneSpace
    scaleMode=BMM_Additive
    scaleSpace=BCS_BoneSpace
    translationMode=BMM_Additive
    translationSpace=BCS_BoneSpace
    alphaBoolBlend {blendInTime=0.15, blendOption=Linear, blendOutTime=1}
    alphaScaleBiasClamp {bInterpResult=true}
    boneToModify {boneName=Dyn_Master_Belt}
    rotation {pitch=-20}
   }
   "Transform (Modify) Bone - Bone: Dyn_Master_LowerBelt"(Alpha=1, Translation="0,0,0", Rotation="-10,0,0", Scale="1,1,1") {
    rotationMode=BMM_Additive
    rotationSpace=BCS_BoneSpace
    scaleMode=BMM_Additive
    scaleSpace=BCS_BoneSpace
    translationMode=BMM_Additive
    translationSpace=BCS_BoneSpace
    alphaBoolBlend {blendInTime=0.15, blendOption=Linear, blendOutTime=1}
    alphaScaleBiasClamp {interpSpeedDecreasing=1, interpSpeedIncreasing=20}
    boneToModify {boneName=Dyn_Master_LowerBelt}
    rotation {pitch=-10}
   }
   "Transform (Modify) Bone - Bone: Dyn_Camera"!(ComponentPose, Alpha:float=1, Translation="-4,4.5,2", Rotation="-15,-20,60", Scale="1,1,1") -> (Pose) {
    rotationMode=BMM_Additive
    rotationSpace=BCS_BoneSpace
    scaleMode=BMM_Additive
    scaleSpace=BCS_BoneSpace
    translationMode=BMM_Additive
    translationSpace=BCS_BoneSpace
    alphaBoolBlend {blendInTime=0.15, blendOption=Linear, blendOutTime=1}
    alphaScaleBiasClamp {interpSpeedDecreasing=1, interpSpeedIncreasing=20}
    boneToModify {boneName=Dyn_Camera}
    rotation=<rot -15 -20 60>
    translation=<-4 4.5 2>
   }
   "Transform (Modify) Bone - Bone: Dyn_CameraBelt_Side"!(ComponentPose, Alpha:float=1, Translation="-2,1,0", Rotation="0,0,0", Scale="1,1,1") -> (Pose) {
    rotationMode=BMM_Additive
    rotationSpace=BCS_BoneSpace
    scaleMode=BMM_Additive
    scaleSpace=BCS_BoneSpace
    translationMode=BMM_Additive
    translationSpace=BCS_BoneSpace
    alphaBoolBlend {blendInTime=0.15, blendOption=Linear, blendOutTime=1}
    alphaScaleBiasClamp {interpSpeedDecreasing=1, interpSpeedIncreasing=20}
    boneToModify {boneName=Dyn_CameraBelt_Side}
    translation=<-2 1>
   }
   "Transform (Modify) Bone - Bone: Dyn_Pouch3"(Alpha=1, Translation="0,0.000,-5.000", Rotation="0.000,0.000,-20.000", Scale="1,1,1") {
    rotationMode=BMM_Additive
    rotationSpace=BCS_BoneSpace
    scaleMode=BMM_Additive
    scaleSpace=BCS_BoneSpace
    translationMode=BMM_Additive
    translationSpace=BCS_BoneSpace
    alphaBoolBlend {blendInTime=0.05, blendOption=Linear, blendOutTime=0.1}
    alphaScaleBiasClamp {interpSpeedDecreasing=1, interpSpeedIncreasing=0.5}
    boneToModify {boneName=Dyn_Pouch3}
    rotation {roll=-20}
    translation {z=-5}
   }
   "Transform (Modify) Bone - Bone: Dyn_Pouch3"(Alpha=1, Translation="-1,0.000,0", Rotation="0.000,0.000,50", Scale="1,1,1") {
    rotationMode=BMM_Additive
    rotationSpace=BCS_BoneSpace
    scaleMode=BMM_Additive
    scaleSpace=BCS_BoneSpace
    translationMode=BMM_Additive
    translationSpace=BCS_BoneSpace
    alphaBoolBlend {blendInTime=0.05, blendOption=Linear, blendOutTime=0.1}
    alphaScaleBiasClamp {interpSpeedDecreasing=1, interpSpeedIncreasing=0.5}
    boneToModify {boneName=Dyn_Pouch3}
    rotation {roll=50}
    translation {x=-1}
   }
   n25 = "Transform (Modify) Bone - Bone: Dyn_Master_LowerBelt"(Alpha=1, Translation="0,0,0", Rotation="-15,0,0", Scale="1,1,1") {
    rotationMode=BMM_Additive
    rotationSpace=BCS_BoneSpace
    scaleMode=BMM_Additive
    scaleSpace=BCS_BoneSpace
    translationMode=BMM_Additive
    translationSpace=BCS_BoneSpace
    alphaBoolBlend {blendInTime=0.15, blendOption=Linear, blendOutTime=1}
    alphaScaleBiasClamp {interpSpeedDecreasing=1, interpSpeedIncreasing=20}
    boneToModify {boneName=Dyn_Master_LowerBelt}
    rotation {pitch=-15}
   }
   n26 = "Transform (Modify) Bone - Bone: Dyn_Camera"(ComponentPose=@n25, Alpha=1, Translation="-1,0,0", Rotation="0,-5,0", Scale="1,1,1") {
    rotationMode=BMM_Additive
    rotationSpace=BCS_BoneSpace
    scaleMode=BMM_Additive
    scaleSpace=BCS_BoneSpace
    translationMode=BMM_Additive
    translationSpace=BCS_BoneSpace
    alphaBoolBlend {blendInTime=0.15, blendOption=Linear, blendOutTime=1}
    alphaScaleBiasClamp {interpSpeedDecreasing=1, interpSpeedIncreasing=20}
    boneToModify {boneName=Dyn_Camera}
    rotation {yaw=-5}
    translation {x=-1}
   }
   n27 = "Transform (Modify) Bone - Bone: Dyn_CameraBelt_Side"(ComponentPose=@n26, Alpha=1, Translation="-1,0.8,0", Rotation="0,0,0", Scale="1,1,1") {
    rotationMode=BMM_Additive
    rotationSpace=BCS_BoneSpace
    scaleMode=BMM_Additive
    scaleSpace=BCS_BoneSpace
    translationMode=BMM_Additive
    translationSpace=BCS_BoneSpace
    alphaBoolBlend {blendInTime=0.15, blendOption=Linear, blendOutTime=1}
    alphaScaleBiasClamp {interpSpeedDecreasing=1, interpSpeedIncreasing=20}
    boneToModify {boneName=Dyn_CameraBelt_Side}
    translation=<-1 0.8>
   }
   "Transform (Modify) Bone - Bone: Dyn_Master_Belt_RightFront"(ComponentPose=@n27, Alpha=1, Translation="-1,0,0", Rotation="0,0,0", Scale="1,1,1") {
    rotationMode=BMM_Additive
    rotationSpace=BCS_BoneSpace
    scaleMode=BMM_Additive
    scaleSpace=BCS_BoneSpace
    translationMode=BMM_Additive
    translationSpace=BCS_BoneSpace
    alphaBoolBlend {blendInTime=0.15, blendOption=Linear, blendOutTime=1}
    alphaScaleBiasClamp {interpSpeedDecreasing=1, interpSpeedIncreasing=20}
    boneToModify {boneName=Dyn_Master_Belt_RightFront}
    translation {x=-1}
   }
   "Transform (Modify) Bone - Bone: Dyn_Master_Pouch"(Alpha=1, Translation="-1,0,0", Rotation="0,0,-15", Scale="1,1,1") {
    rotationMode=BMM_Additive
    rotationSpace=BCS_BoneSpace
    scaleMode=BMM_Additive
    scaleSpace=BCS_BoneSpace
    translationMode=BMM_Additive
    translationSpace=BCS_BoneSpace
    alphaBoolBlend {blendInTime=0.05, blendOption=Linear, blendOutTime=0.1}
    alphaScaleBiasClamp {bInterpResult=true, interpSpeedDecreasing=1, interpSpeedIncreasing=0.5}
    boneToModify {boneName=Dyn_Master_Pouch}
    rotation {roll=-15}
    translation {x=-1}
   }
   n28 = "Constraint - Dyn_Pouch2 (PPPPP)"!(
    ComponentPose=Local To Component(LocalPose=Use cached pose 'PoseCache'())
    Alpha=1
    ConstraintWeights_0=0.0003
    ConstraintWeights_1:float=0.0033
    ConstraintWeights_2:float=0.0002
    ConstraintWeights_3:float=0.2229
    ConstraintWeights_4:float=0.7734
   ) -> (Pose) {
    alphaBoolBlend {blendOption=Linear}
    boneToModify {boneName=Dyn_Pouch2}
    constraintSetup [
     {targetBone {boneName=spine_02}}
     {targetBone {boneName=thigh_l}}
     {targetBone {boneName=thigh_twist_01_r}}
     {targetBone {boneName=thigh_twist_01_l}}
     {targetBone {boneName=pelvis}}
    ]
    constraintWeights [1,1,1,1,1]
   }
   n29 = "Constraint - Dyn_Pouch3 (PPPPP)"!(
    ComponentPose=@n28
    Alpha=1
    ConstraintWeights_0=0
    ConstraintWeights_1:float=0.0218
    ConstraintWeights_2:float=0.0007
    ConstraintWeights_3:float=0.5136
    ConstraintWeights_4:float=0.4639
   ) -> (Pose) {
    alpha=0
    alphaBoolBlend {blendOption=Linear}
    boneToModify {boneName=Dyn_Pouch3}
    constraintSetup [
     {targetBone {boneName=spine_02}}
     {targetBone {boneName=thigh_l}}
     {targetBone {boneName=thigh_twist_01_r}}
     {targetBone {boneName=thigh_twist_01_l}}
     {targetBone {boneName=pelvis}}
    ]
    constraintWeights [1,1,1,1,1]
   }
   n30 = "Transform (Modify) Bone - Bone: Dyn_Pouch2"(ComponentPose=@n29, Alpha=1, Translation="-1.140,-0.879,0.036", Rotation="-18.762,1.914,-50.684", Scale="1,1,1") {
    rotationMode=BMM_Replace
    rotationSpace=BCS_BoneSpace
    scaleMode=BMM_Additive
    scaleSpace=BCS_BoneSpace
    translationMode=BMM_Additive
    translationSpace=BCS_BoneSpace
    alphaBoolBlend {blendInTime=0.05, blendOption=Linear, blendOutTime=0.1}
    alphaScaleBiasClamp {bInterpResult=true, interpSpeedDecreasing=1, interpSpeedIncreasing=0.5}
    boneToModify {boneName=Dyn_Pouch2}
    rotation=<rot -18.762 1.914 -50.6842>
    translation=<-1.14 -0.879 0.036>
   }
   Output Pose(Result=Blend Poses by bool(BlendPose_0=Use cached pose 'PoseCache'(), BlendPose_1=Component To Local(ComponentPose=@n30), bActiveValue=Get bRigidBodyEnabled()) {blendType=Linear}) {layerGroup=None}
   n31 = "Transform (Modify) Bone - Bone: Dyn_Master_LowerBelt"(
    ComponentPose=Local To Component(LocalPose=Copy Pose From Mesh() {bUseAttachedParent=true})
    Alpha=1
    Translation="0,0,0"
    Rotation="-10,0,0"
    Scale="1,1,1"
   ) {
    rotationMode=BMM_Additive
    rotationSpace=BCS_BoneSpace
    scaleMode=BMM_Additive
    scaleSpace=BCS_BoneSpace
    translationMode=BMM_Additive
    translationSpace=BCS_BoneSpace
    alphaBoolBlend {blendInTime=0.15, blendOption=Linear, blendOutTime=1}
    alphaScaleBiasClamp {interpSpeedDecreasing=1, interpSpeedIncreasing=20}
    boneToModify {boneName=Dyn_Master_LowerBelt}
    rotation {pitch=-10}
   }
   n32 = Get Crouch Sprinting() -> (Crouch Sprinting:double)
   n33 = "Transform (Modify) Bone - Bone: Dyn_Camera"(ComponentPose=@n31, Alpha=@n32, Translation="-4,4.5,4", Rotation="-15,-20,50", Scale="1,1,1") {
    rotationMode=BMM_Additive
    rotationSpace=BCS_BoneSpace
    scaleMode=BMM_Additive
    scaleSpace=BCS_BoneSpace
    translationMode=BMM_Additive
    translationSpace=BCS_BoneSpace
    alphaBoolBlend {blendInTime=0.15, blendOption=Linear, blendOutTime=1}
    alphaScaleBiasClamp {interpSpeedDecreasing=1, interpSpeedIncreasing=20}
    boneToModify {boneName=Dyn_Camera}
    rotation=<rot -15 -20 50>
    translation=<-4 4.5 4>
   }
   n34 = "Transform (Modify) Bone - Bone: Dyn_CameraBelt_Back"(ComponentPose=@n33, Alpha=@n32, Translation="-2,1,0", Rotation="5,0,0", Scale="1,1,1") {
    rotationMode=BMM_Additive
    rotationSpace=BCS_BoneSpace
    scaleMode=BMM_Additive
    scaleSpace=BCS_BoneSpace
    translationMode=BMM_Additive
    translationSpace=BCS_BoneSpace
    alphaBoolBlend {blendInTime=0.15, blendOption=Linear, blendOutTime=1}
    alphaScaleBiasClamp {interpSpeedDecreasing=1, interpSpeedIncreasing=20}
    boneToModify {boneName=Dyn_CameraBelt_Back}
    rotation {pitch=5}
    translation=<-2 1>
   }
   n35 = "Transform (Modify) Bone - Bone: Dyn_CameraBelt_Side"(ComponentPose=@n34, Alpha=@n32, Translation="-2,1,0", Rotation="0,0,0", Scale="1,1,1") {
    rotationMode=BMM_Additive
    rotationSpace=BCS_BoneSpace
    scaleMode=BMM_Additive
    scaleSpace=BCS_BoneSpace
    translationMode=BMM_Additive
    translationSpace=BCS_BoneSpace
    alphaBoolBlend {blendInTime=0.15, blendOption=Linear, blendOutTime=1}
    alphaScaleBiasClamp {interpSpeedDecreasing=1, interpSpeedIncreasing=20}
    boneToModify {boneName=Dyn_CameraBelt_Side}
    translation=<-2 1>
   }
   n36 = "Transform (Modify) Bone - Bone: Dyn_Master_UpperBelt"(ComponentPose=@n35, Alpha=@n32, Translation="0,0,0", Rotation="-5,0,0", Scale="1,1,1") {
    rotationMode=BMM_Additive
    rotationSpace=BCS_BoneSpace
    scaleMode=BMM_Additive
    scaleSpace=BCS_BoneSpace
    translationMode=BMM_Additive
    translationSpace=BCS_BoneSpace
    alphaBoolBlend {blendInTime=0.15, blendOption=Linear, blendOutTime=1}
    alphaScaleBiasClamp {interpSpeedDecreasing=1, interpSpeedIncreasing=20}
    boneToModify {boneName=Dyn_Master_UpperBelt}
    rotation {pitch=-5}
   }
   n37 = "Transform (Modify) Bone - Bone: Dyn_Master_Belt_RightFront"(ComponentPose=@n36, Alpha=@n32, Translation="-3,0,0", Rotation="10,0,0", Scale="1,1,1") {
    rotationMode=BMM_Additive
    rotationSpace=BCS_BoneSpace
    scaleMode=BMM_Additive
    scaleSpace=BCS_BoneSpace
    translationMode=BMM_Additive
    translationSpace=BCS_BoneSpace
    alphaBoolBlend {blendInTime=0.15, blendOption=Linear, blendOutTime=1}
    alphaScaleBiasClamp {interpSpeedDecreasing=1, interpSpeedIncreasing=20}
    boneToModify {boneName=Dyn_Master_Belt_RightFront}
    rotation {pitch=10}
    translation {x=-3}
   }
   n38 = "Transform (Modify) Bone - Bone: Dyn_Master_Pouch"(ComponentPose=@n37, Alpha=1, Translation="-2,3,-7", Rotation="0.000,0.000,-80", Scale="1,1,1") {
    rotationMode=BMM_Additive
    rotationSpace=BCS_BoneSpace
    scaleMode=BMM_Additive
    scaleSpace=BCS_BoneSpace
    translationMode=BMM_Additive
    translationSpace=BCS_BoneSpace
    alphaBoolBlend {blendInTime=0.05, blendOption=Linear, blendOutTime=0.1}
    alphaScaleBiasClamp {interpSpeedDecreasing=1, interpSpeedIncreasing=0.5}
    boneToModify {boneName=Dyn_Master_Pouch}
    rotation {roll=-80}
    translation=<-2 3 -7>
   }
   n39 = "Transform (Modify) Bone - Bone: Dyn_Beads_l1"(ComponentPose=@n38, Alpha=1, Translation="5.000,2.000,3", Rotation="-40,-10,0.000", Scale="1,1,1") {
    rotationMode=BMM_Additive
    rotationSpace=BCS_BoneSpace
    scaleMode=BMM_Additive
    scaleSpace=BCS_BoneSpace
    translationMode=BMM_Additive
    translationSpace=BCS_BoneSpace
    alphaBoolBlend {blendInTime=0.05, blendOption=Linear, blendOutTime=0.1}
    alphaScaleBiasClamp {interpSpeedDecreasing=1, interpSpeedIncreasing=0.5}
    boneToModify {boneName=Dyn_Beads_l1}
    rotation {pitch=-40, yaw=-10}
    translation=<5 2 3>
   }
   n40 = "Transform (Modify) Bone - Bone: Dyn_Beads_l2"(ComponentPose=@n39, Alpha=1, Translation="-2,0,0", Rotation="0,0,0", Scale="1,1,1") {
    rotationMode=BMM_Additive
    rotationSpace=BCS_BoneSpace
    scaleMode=BMM_Additive
    scaleSpace=BCS_BoneSpace
    translationMode=BMM_Additive
    translationSpace=BCS_BoneSpace
    alphaBoolBlend {blendInTime=0.05, blendOption=Linear, blendOutTime=0.1}
    alphaScaleBiasClamp {interpSpeedDecreasing=1, interpSpeedIncreasing=0.5}
    boneToModify {boneName=Dyn_Beads_l2}
    translation {x=-2}
   }
   n41 = "Transform (Modify) Bone - Bone: Dyn_Beads_r1"(ComponentPose=@n40, Alpha=1, Translation="-1,1,-2", Rotation="20,0.000,0.000", Scale="1,1,1") {
    rotationMode=BMM_Additive
    rotationSpace=BCS_BoneSpace
    scaleMode=BMM_Additive
    scaleSpace=BCS_BoneSpace
    translationMode=BMM_Additive
    translationSpace=BCS_BoneSpace
    alphaBoolBlend {blendInTime=0.05, blendOption=Linear, blendOutTime=0.1}
    alphaScaleBiasClamp {interpSpeedDecreasing=1, interpSpeedIncreasing=0.5}
    boneToModify {boneName=Dyn_Beads_r1}
    rotation {pitch=20}
    translation=<-1 1 -2>
   }
   n42 = "Transform (Modify) Bone - Bone: Dyn_Pouch3"(ComponentPose=@n41, Alpha=1, Translation="-0.5,0.000,0", Rotation="0,0,0", Scale="1,1,1") {
    rotationMode=BMM_Additive
    rotationSpace=BCS_BoneSpace
    scaleMode=BMM_Additive
    scaleSpace=BCS_BoneSpace
    translationMode=BMM_Additive
    translationSpace=BCS_BoneSpace
    alphaBoolBlend {blendInTime=0.05, blendOption=Linear, blendOutTime=0.1}
    alphaScaleBiasClamp {interpSpeedDecreasing=1, interpSpeedIncreasing=0.5}
    boneToModify {boneName=Dyn_Pouch3}
    translation {x=-0.5}
   }
   n43 = "Transform (Modify) Bone - Bone: Dyn_Master_LowerBelt"(
    ComponentPose=Local To Component(LocalPose=Copy Pose From Mesh() {bUseAttachedParent=true})
    Alpha=1
    Translation="0,0,0"
    Rotation="-20,0,0"
    Scale="1,1,1"
   ) {
    rotationMode=BMM_Additive
    rotationSpace=BCS_BoneSpace
    scaleMode=BMM_Additive
    scaleSpace=BCS_BoneSpace
    translationMode=BMM_Additive
    translationSpace=BCS_BoneSpace
    alphaBoolBlend {blendInTime=0.15, blendOption=Linear, blendOutTime=1}
    alphaScaleBiasClamp {interpSpeedDecreasing=1, interpSpeedIncreasing=20}
    boneToModify {boneName=Dyn_Master_LowerBelt}
    rotation {pitch=-20}
   }
   n44 = Get CrouchWalking() -> (CrouchWalking:double)
   n45 = "Transform (Modify) Bone - Bone: Dyn_Camera"(ComponentPose=@n43, Alpha=@n44, Translation="-5.5,2,4", Rotation="-20,-20,45", Scale="1,1,1") {
    rotationMode=BMM_Additive
    rotationSpace=BCS_BoneSpace
    scaleMode=BMM_Additive
    scaleSpace=BCS_BoneSpace
    translationMode=BMM_Additive
    translationSpace=BCS_BoneSpace
    alphaBoolBlend {blendInTime=0.15, blendOption=Linear, blendOutTime=1}
    alphaScaleBiasClamp {interpSpeedDecreasing=1, interpSpeedIncreasing=20}
    boneToModify {boneName=Dyn_Camera}
    rotation=<rot -20 -20 45>
    translation=<-5.5 2 4>
   }
   n46 = "Transform (Modify) Bone - Bone: Dyn_CameraBelt_Back"(ComponentPose=@n45, Alpha=@n44, Translation="-2,1,0", Rotation="5,0,0", Scale="1,1,1") {
    rotationMode=BMM_Additive
    rotationSpace=BCS_BoneSpace
    scaleMode=BMM_Additive
    scaleSpace=BCS_BoneSpace
    translationMode=BMM_Additive
    translationSpace=BCS_BoneSpace
    alphaBoolBlend {blendInTime=0.15, blendOption=Linear, blendOutTime=1}
    alphaScaleBiasClamp {interpSpeedDecreasing=1, interpSpeedIncreasing=20}
    boneToModify {boneName=Dyn_CameraBelt_Back}
    rotation {pitch=5}
    translation=<-2 1>
   }
   n47 = "Transform (Modify) Bone - Bone: Dyn_CameraBelt_Side"(ComponentPose=@n46, Alpha=@n44, Translation="-5,1,0", Rotation="0,0,0", Scale="1,1,1") {
    rotationMode=BMM_Additive
    rotationSpace=BCS_BoneSpace
    scaleMode=BMM_Additive
    scaleSpace=BCS_BoneSpace
    translationMode=BMM_Additive
    translationSpace=BCS_BoneSpace
    alphaBoolBlend {blendInTime=0.15, blendOption=Linear, blendOutTime=1}
    alphaScaleBiasClamp {interpSpeedDecreasing=1, interpSpeedIncreasing=20}
    boneToModify {boneName=Dyn_CameraBelt_Side}
    translation=<-5 1>
   }
   n48 = "Transform (Modify) Bone - Bone: Dyn_Master_UpperBelt"(ComponentPose=@n47, Alpha=@n44, Translation="0,0,0", Rotation="0,0,0", Scale="1,1,1") {
    rotationMode=BMM_Additive
    rotationSpace=BCS_BoneSpace
    scaleMode=BMM_Additive
    scaleSpace=BCS_BoneSpace
    translationMode=BMM_Additive
    translationSpace=BCS_BoneSpace
    alphaBoolBlend {blendInTime=0.15, blendOption=Linear, blendOutTime=1}
    alphaScaleBiasClamp {interpSpeedDecreasing=1, interpSpeedIncreasing=20}
    boneToModify {boneName=Dyn_Master_UpperBelt}
   }
   n49 = "Transform (Modify) Bone - Bone: Dyn_Master_Belt_RightFront"(ComponentPose=@n48, Alpha=@n44, Translation="-5,0.000,0.000", Rotation="-10,0,0", Scale="1,1,1") {
    rotationMode=BMM_Additive
    rotationSpace=BCS_BoneSpace
    scaleMode=BMM_Additive
    scaleSpace=BCS_BoneSpace
    translationMode=BMM_Additive
    translationSpace=BCS_BoneSpace
    alphaBoolBlend {blendInTime=0.15, blendOption=Linear, blendOutTime=1}
    alphaScaleBiasClamp {interpSpeedDecreasing=1, interpSpeedIncreasing=20}
    boneToModify {boneName=Dyn_Master_Belt_RightFront}
    rotation {pitch=-10}
    translation {x=-5}
   }
   n50 = "Transform (Modify) Bone - Bone: Dyn_Master_Pouch"(ComponentPose=@n49, Alpha=1, Translation="-3,4.2,-7", Rotation="0.000,0.000,-80", Scale="1,1,1") {
    rotationMode=BMM_Additive
    rotationSpace=BCS_BoneSpace
    scaleMode=BMM_Additive
    scaleSpace=BCS_BoneSpace
    translationMode=BMM_Additive
    translationSpace=BCS_BoneSpace
    alphaBoolBlend {blendInTime=0.05, blendOption=Linear, blendOutTime=0.1}
    alphaScaleBiasClamp {interpSpeedDecreasing=1, interpSpeedIncreasing=0.5}
    boneToModify {boneName=Dyn_Master_Pouch}
    rotation {roll=-80}
    translation=<-3 4.2 -7>
   }
   n51 = "Transform (Modify) Bone - Bone: Dyn_Beads_l1"(ComponentPose=@n50, Alpha=1, Translation="5.000,2.000,3", Rotation="-40,-10,0.000", Scale="1,1,1") {
    rotationMode=BMM_Additive
    rotationSpace=BCS_BoneSpace
    scaleMode=BMM_Additive
    scaleSpace=BCS_BoneSpace
    translationMode=BMM_Additive
    translationSpace=BCS_BoneSpace
    alphaBoolBlend {blendInTime=0.05, blendOption=Linear, blendOutTime=0.1}
    alphaScaleBiasClamp {interpSpeedDecreasing=1, interpSpeedIncreasing=0.5}
    boneToModify {boneName=Dyn_Beads_l1}
    rotation {pitch=-40, yaw=-10}
    translation=<5 2 3>
   }
   n52 = Get Left Leg Strafe() -> (Left Leg Strafe:double)
   n53 = "Transform (Modify) Bone - Bone: Dyn_Beads_l2"(ComponentPose=@n51, Alpha=@n52, Translation="0,0,0", Rotation="0,0,0", Scale="1,1,1") {
    rotationMode=BMM_Additive
    rotationSpace=BCS_BoneSpace
    scaleMode=BMM_Additive
    scaleSpace=BCS_BoneSpace
    translationMode=BMM_Additive
    translationSpace=BCS_BoneSpace
    alphaBoolBlend {blendInTime=0.05, blendOption=Linear, blendOutTime=0.1}
    alphaScaleBiasClamp {interpSpeedDecreasing=1, interpSpeedIncreasing=0.5}
    boneToModify {boneName=Dyn_Beads_l2}
   }
   n54 = "Transform (Modify) Bone - Bone: Dyn_Beads_r1"(ComponentPose=@n53, Alpha=1, Translation="-1,1,-2", Rotation="20,0.000,0.000", Scale="1,1,1") {
    rotationMode=BMM_Additive
    rotationSpace=BCS_BoneSpace
    scaleMode=BMM_Additive
    scaleSpace=BCS_BoneSpace
    translationMode=BMM_Additive
    translationSpace=BCS_BoneSpace
    alphaBoolBlend {blendInTime=0.05, blendOption=Linear, blendOutTime=0.1}
    alphaScaleBiasClamp {interpSpeedDecreasing=1, interpSpeedIncreasing=0.5}
    boneToModify {boneName=Dyn_Beads_r1}
    rotation {pitch=20}
    translation=<-1 1 -2>
   }
   n55 = "Transform (Modify) Bone - Bone: Dyn_Pouch3"(ComponentPose=@n54, Alpha=@n52, Translation="-0.5,0.000,0", Rotation="0,0,0", Scale="1,1,1") {
    rotationMode=BMM_Additive
    rotationSpace=BCS_BoneSpace
    scaleMode=BMM_Additive
    scaleSpace=BCS_BoneSpace
    translationMode=BMM_Additive
    translationSpace=BCS_BoneSpace
    alphaBoolBlend {blendInTime=0.05, blendOption=Linear, blendOutTime=0.1}
    alphaScaleBiasClamp {interpSpeedDecreasing=1, interpSpeedIncreasing=0.5}
    boneToModify {boneName=Dyn_Pouch3}
    translation {x=-0.5}
   }
   n56 = Blend Poses by bool(
    BlendPose_0=Component To Local(ComponentPose=@n42)
    BlendPose_1=Component To Local(ComponentPose=@n55)
    BlendTime_0=0.2
    BlendTime_1=0.2
    bActiveValue=Get bIsCrouchSprinting()
   ) {blendType=Linear}
   n57 = "Transform (Modify) Bone - Bone: Dyn_Master_LowerBelt"(
    ComponentPose=Local To Component(LocalPose=Copy Pose From Mesh() {bUseAttachedParent=true})
    Alpha=1
    Translation="0,0,0"
    Rotation="-5,0,0"
    Scale="1,1,1"
   ) {
    rotationMode=BMM_Additive
    rotationSpace=BCS_BoneSpace
    scaleMode=BMM_Additive
    scaleSpace=BCS_BoneSpace
    translationMode=BMM_Additive
    translationSpace=BCS_BoneSpace
    alphaBoolBlend {blendInTime=0.15, blendOption=Linear, blendOutTime=1}
    alphaScaleBiasClamp {interpSpeedDecreasing=5, interpSpeedIncreasing=2}
    boneToModify {boneName=Dyn_Master_LowerBelt}
    rotation {pitch=-5}
   }
   n58 = Get Right Leg Movement() -> (Right Leg Movement:double)
   n59 = "Transform (Modify) Bone - Bone: Dyn_Master_LowerBelt"(ComponentPose=@n57, Alpha=@n58, Translation="0,0,0", Rotation="-20,0,0", Scale="1,1,1") {
    rotationMode=BMM_Additive
    rotationSpace=BCS_BoneSpace
    scaleMode=BMM_Additive
    scaleSpace=BCS_BoneSpace
    translationMode=BMM_Additive
    translationSpace=BCS_BoneSpace
    alphaBoolBlend {blendInTime=0.15, blendOption=Linear, blendOutTime=1}
    alphaScaleBiasClamp {interpSpeedDecreasing=1, interpSpeedIncreasing=20}
    boneToModify {boneName=Dyn_Master_LowerBelt}
    rotation {pitch=-20}
   }
   n60 = "Transform (Modify) Bone - Bone: Dyn_Camera"(ComponentPose=@n59, Alpha=@n58, Translation="-6,0,0", Rotation="10,-50,40", Scale="1,1,1") {
    rotationMode=BMM_Additive
    rotationSpace=BCS_BoneSpace
    scaleMode=BMM_Additive
    scaleSpace=BCS_BoneSpace
    translationMode=BMM_Additive
    translationSpace=BCS_BoneSpace
    alphaBoolBlend {blendInTime=0.15, blendOption=Linear, blendOutTime=1}
    alphaScaleBiasClamp {interpSpeedDecreasing=1, interpSpeedIncreasing=20}
    boneToModify {boneName=Dyn_Camera}
    rotation=<rot 10 -50 40>
    translation {x=-6}
   }
   n61 = "Transform (Modify) Bone - Bone: Dyn_CameraBelt_Side"(ComponentPose=@n60, Alpha=@n58, Translation="-9,0,0", Rotation="-50,50,0", Scale="1,1,1") {
    rotationMode=BMM_Additive
    rotationSpace=BCS_BoneSpace
    scaleMode=BMM_Additive
    scaleSpace=BCS_BoneSpace
    translationMode=BMM_Additive
    translationSpace=BCS_BoneSpace
    alphaBoolBlend {blendInTime=0.15, blendOption=Linear, blendOutTime=1}
    alphaScaleBiasClamp {interpSpeedDecreasing=1, interpSpeedIncreasing=20}
    boneToModify {boneName=Dyn_CameraBelt_Side}
    rotation {pitch=-50, yaw=50}
    translation {x=-9}
   }
   n62 = "Transform (Modify) Bone - Bone: Dyn_Master_Belt_RightFront"(ComponentPose=@n61, Alpha=@n58, Translation="-3,-2,0", Rotation="20,0,0", Scale="1,1,1") {
    rotationMode=BMM_Additive
    rotationSpace=BCS_BoneSpace
    scaleMode=BMM_Additive
    scaleSpace=BCS_BoneSpace
    translationMode=BMM_Additive
    translationSpace=BCS_BoneSpace
    alphaBoolBlend {blendInTime=0.15, blendOption=Linear, blendOutTime=1}
    alphaScaleBiasClamp {interpSpeedDecreasing=1, interpSpeedIncreasing=20}
    boneToModify {boneName=Dyn_Master_Belt_RightFront}
    rotation {pitch=20}
    translation=<-3 -2>
   }
   n63 = "Transform (Modify) Bone - Bone: Dyn_CameraBelt_Back"(ComponentPose=@n62, Alpha=@n58, Translation="-0.1,0,0", Rotation="5,0,0", Scale="1,1,1") {
    rotationMode=BMM_Additive
    rotationSpace=BCS_BoneSpace
    scaleMode=BMM_Additive
    scaleSpace=BCS_BoneSpace
    translationMode=BMM_Additive
    translationSpace=BCS_BoneSpace
    alphaBoolBlend {blendInTime=0.15, blendOption=Linear, blendOutTime=1}
    alphaScaleBiasClamp {interpSpeedDecreasing=1, interpSpeedIncreasing=20}
    boneToModify {boneName=Dyn_CameraBelt_Back}
    rotation {pitch=5}
    translation {x=-0.1}
   }
   n64 = "Transform (Modify) Bone - Bone: Dyn_Master_Pouch"(ComponentPose=@n63, Alpha=1, Translation="0,-2,0", Rotation="0,0,0", Scale="1,1,1") {
    rotationMode=BMM_Additive
    rotationSpace=BCS_BoneSpace
    scaleMode=BMM_Additive
    scaleSpace=BCS_BoneSpace
    translationMode=BMM_Additive
    translationSpace=BCS_BoneSpace
    alphaBoolBlend {blendInTime=0.05, blendOption=Linear, blendOutTime=0.1}
    alphaScaleBiasClamp {interpSpeedDecreasing=1, interpSpeedIncreasing=0.5}
    boneToModify {boneName=Dyn_Master_Pouch}
    translation {y=-2}
   }
   n65 = Get Front_End_Pouch() -> (Front_End_Pouch:double)
   n66 = "Transform (Modify) Bone - Bone: Dyn_Beads_l1"(ComponentPose=@n64, Alpha=@n65, Translation="4.590,-0.219,2.630", Rotation="-44.803,4.313,-13.654", Scale="1,1,1") {
    rotationMode=BMM_Additive
    rotationSpace=BCS_BoneSpace
    scaleMode=BMM_Additive
    scaleSpace=BCS_BoneSpace
    translationMode=BMM_Additive
    translationSpace=BCS_BoneSpace
    alphaBoolBlend {blendInTime=0.05, blendOption=Linear, blendOutTime=0.1}
    alphaScaleBiasClamp {bInterpResult=true, interpSpeedDecreasing=5}
    boneToModify {boneName=Dyn_Beads_l1}
    rotation=<rot -44.803 4.313 -13.654>
    translation=<4.59 -0.219 2.63>
   }
   n67 = "Transform (Modify) Bone - Bone: Dyn_Beads_r1"(ComponentPose=@n66, Alpha=@n65, Translation="0.800,-0.222,0.029", Rotation="0.011,1.874,0.017", Scale="1,1,1") {
    rotationMode=BMM_Additive
    rotationSpace=BCS_BoneSpace
    scaleMode=BMM_Additive
    scaleSpace=BCS_BoneSpace
    translationMode=BMM_Additive
    translationSpace=BCS_BoneSpace
    alphaBoolBlend {blendInTime=0.05, blendOption=Linear, blendOutTime=0.1}
    alphaScaleBiasClamp {bInterpResult=true, interpSpeedDecreasing=1}
    boneToModify {boneName=Dyn_Beads_r1}
    rotation=<rot 0.011 1.874 0.017>
    translation=<0.8 -0.222 0.029>
   }
   n68 = "Transform (Modify) Bone - Bone: Dyn_Master_Pouch"(ComponentPose=@n67, Alpha=1, Translation="-1,2,0", Rotation="-0.002,-0.000,-13.752", Scale="1,1,1") {
    rotationMode=BMM_Additive
    rotationSpace=BCS_BoneSpace
    scaleMode=BMM_Additive
    scaleSpace=BCS_BoneSpace
    translationMode=BMM_Additive
    translationSpace=BCS_BoneSpace
    alphaBoolBlend {blendInTime=0.05, blendOption=Linear, blendOutTime=0.1}
    alphaScaleBiasClamp {bInterpResult=true, interpSpeedDecreasing=1, interpSpeedIncreasing=0.5}
    boneToModify {boneName=Dyn_Master_Pouch}
    rotation=<rot -0.002 -0.0001 -13.752>
    translation=<-1 2>
   }
   n69 = Get Left_Leg Movement()
   n70 = "Transform (Modify) Bone - Bone: Dyn_Beads_l1"(ComponentPose=@n68, Alpha=@n69, Translation="0,2,0", Rotation="0,0,0", Scale="1,1,1") {
    rotationMode=BMM_Additive
    rotationSpace=BCS_BoneSpace
    scaleMode=BMM_Additive
    scaleSpace=BCS_BoneSpace
    translationMode=BMM_Additive
    translationSpace=BCS_BoneSpace
    alphaBoolBlend {blendInTime=0.05, blendOption=Linear, blendOutTime=0.1}
    alphaScaleBiasClamp {interpSpeedDecreasing=1, interpSpeedIncreasing=0.5}
    boneToModify {boneName=Dyn_Beads_l1}
    translation {y=2}
   }
   n71 = "Transform (Modify) Bone - Bone: Dyn_Beads_r1"(ComponentPose=@n70, Alpha=@n69, Translation="0,2,0", Rotation="0,0,0", Scale="1,1,1") {
    rotationMode=BMM_Additive
    rotationSpace=BCS_BoneSpace
    scaleMode=BMM_Additive
    scaleSpace=BCS_BoneSpace
    translationMode=BMM_Additive
    translationSpace=BCS_BoneSpace
    alphaBoolBlend {blendInTime=0.05, blendOption=Linear, blendOutTime=0.1}
    alphaScaleBiasClamp {interpSpeedDecreasing=1, interpSpeedIncreasing=0.5}
    boneToModify {boneName=Dyn_Beads_r1}
    translation {y=2}
   }
   n72 = "Transform (Modify) Bone - Bone: Dyn_Pouch2"(ComponentPose=@n71, Alpha=@n69, Translation="0,0,0", Rotation="0,0,0", Scale="1,1,1") {
    rotationMode=BMM_Additive
    rotationSpace=BCS_BoneSpace
    scaleMode=BMM_Additive
    scaleSpace=BCS_BoneSpace
    translationMode=BMM_Additive
    translationSpace=BCS_BoneSpace
    alphaBoolBlend {blendInTime=0.05, blendOption=Linear, blendOutTime=0.1}
    alphaScaleBiasClamp {interpSpeedDecreasing=1, interpSpeedIncreasing=0.5}
    boneToModify {boneName=Dyn_Pouch2}
   }
   n73 = "Transform (Modify) Bone - Bone: Dyn_Beads_l1"(ComponentPose=@n72, Alpha=@n69, Translation="0,0,0", Rotation="-30,0.000,0.000", Scale="1,1,1") {
    rotationMode=BMM_Additive
    rotationSpace=BCS_BoneSpace
    scaleMode=BMM_Additive
    scaleSpace=BCS_BoneSpace
    translationMode=BMM_Additive
    translationSpace=BCS_BoneSpace
    alphaBoolBlend {blendInTime=0.05, blendOption=Linear, blendOutTime=0.1}
    alphaScaleBiasClamp {interpSpeedDecreasing=1, interpSpeedIncreasing=0.5}
    boneToModify {boneName=Dyn_Beads_l1}
    rotation {pitch=-30}
   }
   n74 = "Transform (Modify) Bone - Bone: Dyn_Beads_r1"(ComponentPose=@n73, Alpha=@n69, Translation="0,0,-2.5", Rotation="-40,0.000,0.000", Scale="1,1,1") {
    rotationMode=BMM_Additive
    rotationSpace=BCS_BoneSpace
    scaleMode=BMM_Additive
    scaleSpace=BCS_BoneSpace
    translationMode=BMM_Additive
    translationSpace=BCS_BoneSpace
    alphaBoolBlend {blendInTime=0.05, blendOption=Linear, blendOutTime=0.1}
    alphaScaleBiasClamp {interpSpeedDecreasing=1, interpSpeedIncreasing=0.5}
    boneToModify {boneName=Dyn_Beads_r1}
    rotation {pitch=-40}
    translation {z=-2.5}
   }
   n75 = Component To Local(ComponentPose=@n74)
   n76 = Get bIsCrouching()
   n77 = "Transform (Modify) Bone - Bone: Dyn_Mask1"(
    ComponentPose=Local To Component(LocalPose=Blend Poses by bool(BlendPose_0=@n56, BlendPose_1=@n75, BlendTime_0=0.2, BlendTime_1=0.2, bActiveValue=@n76) {blendType=Linear})
    Alpha=0.5
    Translation="0,-1,0"
    Rotation="0,0,10"
    Scale="1,1,1"
   ) {
    rotationMode=BMM_Additive
    rotationSpace=BCS_BoneSpace
    scaleMode=BMM_Additive
    scaleSpace=BCS_BoneSpace
    translationMode=BMM_Additive
    translationSpace=BCS_BoneSpace
    alphaBoolBlend {blendInTime=0.15, blendOption=Linear, blendOutTime=1}
    alphaScaleBiasClamp {bInterpResult=true, interpSpeedDecreasing=2, interpSpeedIncreasing=2}
    boneToModify {boneName=Dyn_Mask1}
    rotation {roll=10}
    translation {y=-1}
   }
   n78 = "Transform (Modify) Bone - Bone: Dyn_Collar"(ComponentPose=@n77, Alpha=Get Neck_Up() -> (Neck_Up:double), Translation="0.8,0,0", Rotation="0,0,0", Scale="1,1,1") {
    rotationMode=BMM_Additive
    rotationSpace=BCS_BoneSpace
    scaleMode=BMM_Additive
    scaleSpace=BCS_BoneSpace
    translationMode=BMM_Additive
    translationSpace=BCS_BoneSpace
    alphaBoolBlend {blendInTime=0.15, blendOption=Linear, blendOutTime=1}
    boneToModify {boneName=Dyn_Collar}
    translation {x=0.8}
   }
   n79 = "Transform (Modify) Bone - Bone: Dyn_Collar"!(ComponentPose=@n78, bAlphaBoolEnabled:bool=Get bIsTargeting(), Translation="0.8,0,0", Rotation="0,0,0", Scale="1,1,1") -> (Pose) {
    alphaInputType=Bool
    rotationMode=BMM_Additive
    rotationSpace=BCS_BoneSpace
    scaleMode=BMM_Additive
    scaleSpace=BCS_BoneSpace
    translationMode=BMM_Additive
    translationSpace=BCS_BoneSpace
    alphaBoolBlend {blendInTime=0.15, blendOption=Linear, blendOutTime=1}
    boneToModify {boneName=Dyn_Collar}
    translation {x=0.8}
   }
   n80 = "Transform (Modify) Bone - Bone: Dyn_Master_ShoulderPad"(ComponentPose=@n79, Alpha=Get Shoulder_Armor() -> (Shoulder_Armor:double), Translation="0,0,0", Rotation="0,0,20", Scale="1,1,1") {
    rotationMode=BMM_Additive
    rotationSpace=BCS_BoneSpace
    translationMode=BMM_Additive
    translationSpace=BCS_BoneSpace
    alphaBoolBlend {blendOption=Linear}
    boneToModify {boneName=Dyn_Master_ShoulderPad}
    rotation {roll=20}
   }
   n81 = RigidBody(ComponentPose=@n80) {
    bClampLinearTranslationLimitToRefPose=true
    bFreezeIncomingPoseOnStart=true
    bOverrideWorldGravity=true
    overridePhysicsAsset=&/F_MED_Buff_Cat_Fan_Hip_Physics
    simulationSpace=BaseBoneSpace
    alphaBoolBlend {blendOption=Linear}
    baseBoneRef {boneName=pelvis}
    componentAppliedLinearAccClamp=<1200 1200 1200>
    componentLinearAccScale=<0.1 0.1 0.1>
    componentLinearVelScale=<0.1 0.1 0.1>
    overrideWorldGravity {z=-980}
   }
   n82 = RigidBody(ComponentPose=@n81) {
    bClampLinearTranslationLimitToRefPose=true
    bFreezeIncomingPoseOnStart=true
    bOverrideWorldGravity=true
    overridePhysicsAsset=&/F_MED_Buff_Cat_Fan_Shoe_Physics
    simulationSpace=BaseBoneSpace
    alphaBoolBlend {blendOption=Linear}
    baseBoneRef {boneName=root}
    componentAppliedLinearAccClamp=<1200 1200 1200>
    componentLinearAccScale=<0.1 0.1 0.1>
    componentLinearVelScale=<0.1 0.1 0.1>
    overrideWorldGravity {z=-980}
   }
   Save cached pose 'PoseCache'(Pose=Component To Local(ComponentPose=@n82)) {cacheName=PoseCache}
   n83 = "Transform (Modify) Bone - Bone: Dyn_Master_LowerBelt"(Alpha=1, Translation="0,0,0", Rotation="-15,0,0", Scale="1,1,1") {
    rotationMode=BMM_Additive
    rotationSpace=BCS_BoneSpace
    scaleMode=BMM_Additive
    scaleSpace=BCS_BoneSpace
    translationMode=BMM_Additive
    translationSpace=BCS_BoneSpace
    alphaBoolBlend {blendInTime=0.15, blendOption=Linear, blendOutTime=1}
    alphaScaleBiasClamp {interpSpeedDecreasing=1, interpSpeedIncreasing=20}
    boneToModify {boneName=Dyn_Master_LowerBelt}
    rotation {pitch=-15}
   }
   n84 = "Transform (Modify) Bone - Bone: Dyn_Camera"(ComponentPose=@n83, Alpha=1, Translation="-1,0,0", Rotation="0,-5,0", Scale="1,1,1") {
    rotationMode=BMM_Additive
    rotationSpace=BCS_BoneSpace
    scaleMode=BMM_Additive
    scaleSpace=BCS_BoneSpace
    translationMode=BMM_Additive
    translationSpace=BCS_BoneSpace
    alphaBoolBlend {blendInTime=0.15, blendOption=Linear, blendOutTime=1}
    alphaScaleBiasClamp {interpSpeedDecreasing=1, interpSpeedIncreasing=20}
    boneToModify {boneName=Dyn_Camera}
    rotation {yaw=-5}
    translation {x=-1}
   }
   n85 = "Transform (Modify) Bone - Bone: Dyn_CameraBelt_Side"(ComponentPose=@n84, Alpha=1, Translation="-1,0.8,0", Rotation="0,0,0", Scale="1,1,1") {
    rotationMode=BMM_Additive
    rotationSpace=BCS_BoneSpace
    scaleMode=BMM_Additive
    scaleSpace=BCS_BoneSpace
    translationMode=BMM_Additive
    translationSpace=BCS_BoneSpace
    alphaBoolBlend {blendInTime=0.15, blendOption=Linear, blendOutTime=1}
    alphaScaleBiasClamp {interpSpeedDecreasing=1, interpSpeedIncreasing=20}
    boneToModify {boneName=Dyn_CameraBelt_Side}
    translation=<-1 0.8>
   }
   n86 = "Transform (Modify) Bone - Bone: Dyn_Master_Belt_RightFront"(ComponentPose=@n85, Alpha=1, Translation="-1,0,0", Rotation="0,0,0", Scale="1,1,1") {
    rotationMode=BMM_Additive
    rotationSpace=BCS_BoneSpace
    scaleMode=BMM_Additive
    scaleSpace=BCS_BoneSpace
    translationMode=BMM_Additive
    translationSpace=BCS_BoneSpace
    alphaBoolBlend {blendInTime=0.15, blendOption=Linear, blendOutTime=1}
    alphaScaleBiasClamp {interpSpeedDecreasing=1, interpSpeedIncreasing=20}
    boneToModify {boneName=Dyn_Master_Belt_RightFront}
    translation {x=-1}
   }
   Save cached pose 'Crouch'(Pose=Component To Local(ComponentPose=@n86)) {cacheName=Crouch}
   Save cached pose 'Crouch walk '(Pose=Component To Local()) {cacheName="Crouch walk "}
   Use cached pose 'Stationary'()
   Use cached pose 'Stationary'()
   NOT Boolean(A=false)
   AND Boolean(A=Get bIsJumping(), B=Get bIsSprinting())
   Get bIsFalling()
   Get bIsJumping()
   Get CrouchWalking()
  ]}
 EventGraph {nodes [
   Event BlueprintUpdateAnimation()
   > n1 = Sequence!(execute) -> (then_0, then_1, then_2:Exec, then_3:Exec, then_4:Exec, then_5:Exec, then_6:Exec)
   > n2 = Branch(Condition=Get bIsCrouching())
   > n3 = Branch(Condition=Get bIsCrouchSprinting())
   > Set Crouch Sprinting(Crouch Sprinting:double=GetAlphaFromReferenceTransformDelta(SkeletalMeshComponent=Get Skel Mesh(), BoneName=pelvis, BaseName=thigh_r, InputRange=(X=70,Y=110), TwistAxis="1,0,0")) -> (Output_Get:double)
   Set CrouchWalking(
    execute=@n3.else
    CrouchWalking:double=GetAlphaFromReferenceTransformDelta(SkeletalMeshComponent=Get Skel Mesh(), BoneName=pelvis, BaseName=thigh_r, InputRange=(X=26,Y=100), TwistAxis="1,0,0")
   ) -> (Output_Get:double)
   Set Right Leg Movement(
    execute=@n2.else
    Right Leg Movement:double=GetAlphaFromReferenceTransformDelta(SkeletalMeshComponent=Get Skel Mesh(), BoneName=pelvis, BaseName=thigh_r, InputRange=(X=-1,Y=87), TwistAxis="1,0,0")
   ) -> (Output_Get:double)
   Set Neck_Up(
    execute=@n1.then_1
    Neck_Up:double=GetAlphaFromReferenceTransformDelta(
     SkeletalMeshComponent=Get Skel Mesh()
     BoneName=neck_02
     BaseName=neck_01
     DeltaComponentType=QuaternionTwist
     InputRange=(X=0,Y=0.5)
     TwistAxis="0,0,1"
    )
   ) -> (Output_Get:double)
   Set Shoulder_Armor(
    execute=[@n1.then_2, @n1.then_3]
    Shoulder_Armor:double=GetAlphaFromReferenceTransformDelta(
     SkeletalMeshComponent=Get Skel Mesh()
     BoneName=clavicle_l
     BaseName=upperarm_l
     DeltaComponentType=QuaternionTwist
     InputRange=(X=7,Y=103)
     TwistAxis="1,0,0"
    )
   ) -> (Output_Get:double)
   n4 = Branch(execute=@n1.then_4, Condition=Get bIsInFrontEnd())
   > Set Front_End_Pouch(Front_End_Pouch:double=GetAlphaFromReferenceTransformDelta(SkeletalMeshComponent=Get Skel Mesh(), BoneName=pelvis, BaseName=thigh_l, InputRange=(X=3,Y=-6), TwistAxis="1,0,0")) -> (Output_Get:double)
   n5 = Branch(execute=@n4.else, Condition=Get bIsCrouching())
   > Set Left Leg Strafe(Left Leg Strafe:double=GetAlphaFromReferenceTransformDelta(SkeletalMeshComponent=Get Skel Mesh(), BoneName=pelvis, BaseName=thigh_l, InputRange=(X=105,Y=66), TwistAxis="1,0,0")) -> (Output_Get:double)
   Set Left_Leg Movement(
    execute=@n5.else
    Left_Leg Movement:double=GetAlphaFromReferenceTransformDelta(SkeletalMeshComponent=Get Skel Mesh(), BoneName=pelvis, BaseName=thigh_l, InputRange=(X=21,Y=80), TwistAxis="1,0,0")
   ) -> (Output_Get:double)
   Event BlueprintInitializeAnimation()
   > Set Skel Mesh(Skel Mesh=GetOwnerSkeletalMeshComponent())
   Set Left_Leg Movement(Left_Leg Movement=0)
   Set Left_Leg Movement(Left_Leg Movement=0)
   Set RightLegStrafe(RightLegStrafe:double=GetAlphaFromReferenceTransformDelta(
     SkeletalMeshComponent=Get Skel Mesh()
     BoneName=pelvis
     BaseName=thigh_r
     DeltaComponentType=QuaternionTwist
     InputRange=(X=-42,Y=-170)
     TwistAxis="1,0,0"
    )) -> (Output_Get:double)
   Set CrouchMovement_true(CrouchMovement_true:double=0) -> (Output_Get:double)
   Set CrouchMovement_true(CrouchMovement_true=1)
  ]}
]
variables [
 Crouch Sprinting {type=number}
 CrouchMovement_true {type=number}
 CrouchWalking {type=number}
 Front_End_Pouch {type=number}
 Left Leg Strafe {type=number}
 Left_Leg Movement {type=number}
 Neck_Up {type=number}
 Right Leg Movement {type=number}
 RightLegStrafe {type=number}
 Shoulder_Armor {type=number}
 Skel Mesh {type=SkeletalMeshComponent}
]
